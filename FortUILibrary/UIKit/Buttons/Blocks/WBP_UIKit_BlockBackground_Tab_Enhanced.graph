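# UIKit tab button background block that renders state visuals with a Backing image and a FocusOutline image.
# State transition events play animations on material parameters (hover, focus, press, disable) and the outline, with an optional glow when AdditiveBorder is set.
# Selection reuses the Focused parameter, SelectionIsAvailable gates focus replay, and press animations play only when non-selectable.
# Instant transitions stop animations and write dynamic material parameters directly.
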WidgetBlueprint /FortUILibrary/UIKit/Buttons/Blocks/WBP_UIKit_BlockBackground_Tab_Enhanced
animations [
 Anim_Focused_Transition {
  binding_count=1
  length_seconds=1
  object=&self:Anim_Focused_Transition
  sequence {
   bindings [Backing {
     id=#1
     possessed_class=Image
     tracks [{
       type=MovieSceneWidgetMaterialTrack
       sections [{
         type=MovieSceneParameterSection
         end_frame=120
         start_frame=0
         channels [Trans_Focused {type=MovieSceneScriptingFloatChannel, keys [[0,0],[120,1]]}]
        }]
      }]
    }]
   display_rate {numerator=120}
  }
 }
 Anim_Pressed_Transition {
  binding_count=1
  length_seconds=1
  object=&self:Anim_Pressed_Transition
  sequence {bindings [Backing {
     id=#2
     possessed_class=Image
     tracks [{
       type=MovieSceneWidgetMaterialTrack
       sections [{
         type=MovieSceneParameterSection
         end_frame=20
         start_frame=0
         channels [Trans_Pressed {type=MovieSceneScriptingFloatChannel, keys [[0,0],[19.9997,1]]}]
        }]
      }]
    }]}
 }
 Anim_Disabled_Transition {
  binding_count=1
  length_seconds=1
  object=&self:Anim_Disabled_Transition
  sequence {bindings [Backing {
     id=#3
     possessed_class=Image
     tracks [{
       type=MovieSceneWidgetMaterialTrack
       sections [{
         type=MovieSceneParameterSection
         end_frame=20
         start_frame=0
         channels [Trans_Disabled {type=MovieSceneScriptingFloatChannel, keys [[0,0],[20,1]]}]
        }]
      }]
    }]}
 }
 Anim_Outline {
  binding_count=3
  length_seconds=1
  object=&self:Anim_Outline
  sequence {
   bindings [
    OverlaySlot_7 {id=#4, parent_id=#5, possessed_class=OverlaySlot, tracks []}
    FocusOutline {
     id=#5
     possessed_class=Image
     tracks [
      {
       type=MovieSceneFloatTrack
       display_name=Render Opacity
       sections [{
         type=MovieSceneFloatSection
         end_frame=120
         start_frame=0
         channels [None {type=MovieSceneScriptingFloatChannel, keys [[0,0],[20,1],[120,1]]}]
        }]
      }
      {
       type=MovieSceneWidgetMaterialTrack
       sections [{type=MovieSceneParameterSection, channels [BorderOffset_Anim {type=MovieSceneScriptingFloatChannel, keys [[0,1],[84,0],[120,0]]}]}]
      }
     ]
    }
    Backing {
     id=#6
     possessed_class=Image
     tracks [{
       type=MovieSceneWidgetMaterialTrack
       sections [{
         type=MovieSceneParameterSection
         end_frame=120
         start_frame=0
         channels [Trans_Hovered {type=MovieSceneScriptingFloatChannel, keys [[0,0],[20,1]]}]
        }]
      }]
    }
   ]
   display_rate {numerator=120}
  }
 }
 Anim_Outline_Glow {
  binding_count=1
  length_seconds=2.4963
  object=&self:Anim_Outline_Glow
  sequence {bindings [FocusOutline {
     id=#7
     possessed_class=Image
     tracks [{
       type=MovieSceneWidgetMaterialTrack
       sections [{
         type=MovieSceneParameterSection
         end_frame=49
         start_frame=0
         channels [NoiseOn_Anim {type=MovieSceneScriptingFloatChannel, keys [[0,0],[5,0],[49.9253,1]]}]
        }]
      }]
    }]}
 }
]
graphs [
 EventGraph {nodes [
   Event OnInstantTransitionToSelected() {is=title K2Node_Event}
   > StopAllAnimations(self:User Widget Object Reference)
   > SetRenderOpacity(self=Get FocusOutline() -> (FocusOutline:Image Object Reference), InOpacity=0)
   n1 = Get BackgroundDynamicMaterial() -> (BackgroundDynamicMaterial:Material Instance Dynamic Object Reference)
   > InstantTransitionToEnabled!(execute, MID:Material Instance Dynamic Object Reference=@n1, IsSelected:bool=true) -> (then)
   > SetScalarParameterValue(self=@n1, ParameterName=Trans_Focused, Value=1)
   Event OnTransitionSelectedRequested() {is=title K2Node_Event}
   > PlayTransition!(
    execute
    self:UIKit Block Object Reference
    Animation:Widget Animation Object Reference=Get Anim_Focused_Transition() -> (Anim_Focused_Transition:Widget Animation Object Reference)
    bPlayForward:bool=true
   ) -> (then)
   > Set SelectionIsAvailable(SelectionIsAvailable:bool=false) -> (Output_Get:bool)
   Event OnTransitionFocusedRequested() {is=title K2Node_Event}
   > n2 = Branch(Condition=GetIsSelectable!() -> (IsSelectable:bool))
   > n3 = Branch(Condition=GetIsSelected!() -> (IsSelected:bool))
   > n4 = Branch(Condition=Get SelectionIsAvailable() -> (SelectionIsAvailable:bool))
   > PlayTransition(Animation=Get Anim_Focused_Transition(), bPlayForward=true)
   > Set SelectionIsAvailable(SelectionIsAvailable=false)
   > StopAnimation(InAnimation=Get Anim_Outline() -> (Anim_Outline:Widget Animation Object Reference))
   > StopAnimation(InAnimation=Get Anim_Outline_Glow() -> (Anim_Outline_Glow:Widget Animation Object Reference))
   > SetRenderOpacity(self=Get FocusOutline(), InOpacity=0)
   n5 = PlayTransition(execute=[@n6.then, @n4.else, @n3.else, @n11.else], Animation=Get Anim_Outline(), bPlayForward=true)
   Branch(execute=[@n5.then, @n11.then], Condition=Get AdditiveBorder() -> (AdditiveBorder:bool))
   > PlayAnimationForward(InAnimation=Get Anim_Outline_Glow())
   n6 = PlayTransition(execute=@n2.else, Animation=Get Anim_Focused_Transition(), bPlayForward=true)
   Event OnTransitionUnfocusedRequested() {is=title K2Node_Event}
   > StopAnimation(InAnimation=Get Anim_Outline())
   > StopAnimation(InAnimation=Get Anim_Outline_Glow())
   > SetRenderOpacity(self=Get FocusOutline(), InOpacity=0)
   > UnhoverBackgroundReset()
   > n7 = Branch(Condition=GetIsSelectable())
   PlayTransition(execute=@n7.else, Animation=Get Anim_Focused_Transition(), bPlayForward=false)
   Event OnTransitionPressedRequested() {is=title K2Node_Event}
   > n8 = Branch(Condition=GetIsSelectable())
   PlayTransition(execute=@n8.else, Animation=Get Anim_Pressed_Transition() -> (Anim_Pressed_Transition:Widget Animation Object Reference), bPlayForward=true)
   Event OnTransitionReleasedRequested() {is=title K2Node_Event}
   > n9 = Branch(Condition=GetIsSelectable())
   PlayTransition(execute=@n9.else, Animation=Get Anim_Pressed_Transition(), bPlayForward=false)
   Event OnTransitionDisabledRequested() {is=title K2Node_Event}
   > StopAnimation(InAnimation=Get Anim_Outline())
   > StopAnimation(InAnimation=Get Anim_Outline_Glow())
   > SetRenderOpacity(self=Get FocusOutline(), InOpacity=0)
   > UnhoverBackgroundReset()
   > PlayTransition(Animation=Get Anim_Disabled_Transition() -> (Anim_Disabled_Transition:Widget Animation Object Reference), bPlayForward=true)
   Event OnTransitionEnabledRequested() {is=title K2Node_Event}
   > PlayTransition(Animation=Get Anim_Disabled_Transition(), bPlayForward=false)
   Event OnTransitionDeselectedIdleRequested() {is=title K2Node_Event}
   > PlayTransition(Animation=Get Anim_Focused_Transition(), bPlayForward=false)
   > SetScalarParameterValue(self=Get BackgroundDynamicMaterial(), ParameterName=Trans_Disabled, Value=0)
   > Set SelectionIsAvailable(SelectionIsAvailable=true)
   Event OnTransitionDeselectedFocusedRequested() {is=title K2Node_Event}
   > PlayTransition(Animation=Get Anim_Focused_Transition(), bPlayForward=false)
   > SetScalarParameterValue(self=Get BackgroundDynamicMaterial(), ParameterName=Trans_Disabled, Value=0)
   > Set SelectionIsAvailable(SelectionIsAvailable=true)
   Event PreConstruct()
   > SetBackgroundMaterial!(
    execute
    self:WBP UIKit Block Background Base Object Reference
    Material Instance:"Material Instance Object Reference (by ref)"=Get BackgroundMaterial() -> (BackgroundMaterial:Material Instance Object Reference)
   ) -> (then)
   > SetOutlineMaterial!(
    execute
    self:WBP UIKit Block Background Base Object Reference
    Material Instance:"Material Instance Object Reference (by ref)"=Get OutlineMaterial() -> (OutlineMaterial:Material Instance Object Reference)
   ) -> (then)
   Event OnInstantTransitionToEnabled() {is=title K2Node_Event}
   > StopAllAnimations()
   > SetRenderOpacity(self=Get FocusOutline(), InOpacity=0)
   > UnhoverBackgroundReset()
   n10 = Get BackgroundDynamicMaterial()
   > SetScalarParameterValue(self=@n10, ParameterName=Trans_Hovered, Value=0)
   > SetScalarParameterValue(self=@n10, ParameterName=Trans_Focused, Value=0)
   > SetScalarParameterValue(self=@n10, ParameterName=Trans_Pressed, Value=0)
   > SetScalarParameterValue(self=@n10, ParameterName=Trans_Disabled, Value=0)
   Event OnInstantTransitionToDisabled() {is=title K2Node_Event}
   > StopAllAnimations()
   > SetRenderOpacity(self=Get FocusOutline(), InOpacity=0)
   > UnhoverBackgroundReset()
   > SetScalarParameterValue(self=Get BackgroundDynamicMaterial(), ParameterName=Trans_Disabled, Value=1)
   n11 = Branch(Condition=GetIsSelected())
  ]}
 SetBackgroundBrush {nodes [
   n1 = SetBackgroundBrush!() -> (then, InBrush:"Slate Brush Structure (by ref)") {is=title K2Node_FunctionEntry}
   n2 = Get Backing() -> (Backing:Image Object Reference)
   > SetBrush(self=@n2, InBrush=@n1)
   > n3 = GetDynamicMaterial(self=@n2)
   > Set BackgroundDynamicMaterial(BackgroundDynamicMaterial:Material Instance Dynamic Object Reference=@n3) -> (Output_Get:Material Instance Dynamic Object Reference)
   > Return Node()
  ]}
 SetBackgroundMaterial {nodes [
   n1 = SetBackgroundMaterial!() -> (then, Material Instance:"Material Instance Object Reference (by ref)") {is=title K2Node_FunctionEntry}
   n2 = Get Backing()
   > SetBrushFromMaterial(self=@n2, Material=@n1)
   > n3 = GetDynamicMaterial(self=@n2)
   > Set BackgroundDynamicMaterial(BackgroundDynamicMaterial=@n3)
   > Return Node()
  ]}
 SetCornerRadiuses {nodes [
   n1 = SetCornerRadiuses!() -> (then, CornerRadius:float=~, RadiusInPixels:bool=~) {is=title K2Node_FunctionEntry}
   > Set Corner Radius(Corner Radius:float=@n1) -> (Output_Get:float)
   > Set Radius in Pixels(Radius in Pixels:bool=@n1.RadiusInPixels) -> (Output_Get:bool)
   n2 = Get OutlineDynamicMaterial() -> (OutlineDynamicMaterial:Material Instance Dynamic Object Reference)
   > SetScalarParameterValue(self=@n2, ParameterName=Radius in Pixels, Value=0)
   > SetVectorParameterValue(
    self=@n2
    ParameterName=Corner Radius
    Value=Make Linear Color!(R:float=0.3, G:float=0.3, B:float=0.3, A:float=0.3) -> (LinearColor:Linear Color Structure) {is=title K2Node_MakeStruct}
   )
   n3 = Get BackgroundDynamicMaterial()
   > SetScalarParameterValue(self=@n3, ParameterName=Radius in Pixels, Value="To Float (Boolean)"(InBool=Get Radius in Pixels() -> (Radius in Pixels:bool)))
   > SetScalarParameterValue(self=@n3, ParameterName=Corner Radius, Value=Get Corner Radius() -> (Corner Radius:float))
  ]}
 SetOutlineMaterial {nodes [
   n1 = SetOutlineMaterial!() -> (then, Material Instance:"Material Instance Object Reference (by ref)") {is=title K2Node_FunctionEntry}
   n2 = Get FocusOutline()
   > SetBrushFromMaterial(self=@n2, Material=@n1)
   > n3 = GetDynamicMaterial(self=@n2)
   > Set OutlineDynamicMaterial(OutlineDynamicMaterial:Material Instance Dynamic Object Reference=@n3) -> (Output_Get:Material Instance Dynamic Object Reference)
   > Return Node()
  ]}
 UnhoverBackgroundReset {nodes [
   UnhoverBackgroundReset() {is=title K2Node_FunctionEntry}
   > SetScalarParameterValue(self=Get BackgroundDynamicMaterial(), ParameterName=Trans_Hovered, Value=0)
  ]}
]
variables [
 AdditiveBorder {type=boolean, default=true}
 Corner Radius {type=number, default=0.325}
 Radius in Pixels {type=boolean}
 SelectionIsAvailable {type=boolean, default=true}
]
widgets {
 info {
  inheritedWidgetCount=0
  parentClass=/FortUILibrary/Systems/ModularButton/Blocks/WBP_UIKit_BlockBackground_Base.WBP_UIKit_BlockBackground_Base_C
  rootWidgetClass=&SizeBox
  widgetCount=4
 }
 widgets [
  SizeBox {
   widgetClassPath=&SizeBox
   widget=SizeBox
   properties {bOverride_HeightOverride=true, bOverride_WidthOverride=true, heightOverride=48, widthOverride=48}
  }
  Root {widgetClassPath=&GridPanel, parent=SizeBox, widget=Root, properties {columnFill [1], rowFill [1]}, slot {class=SizeBoxSlot}}
  FocusOutline {
   widgetClassPath=&Image
   bIsVariable=true
   parent=Root
   widget=FocusOutline
   properties {
    renderOpacity=0
    visibility=SelfHitTestInvisible
    brush {
     bIsDynamicallyLoaded=false
     drawAs=Box
     imageType=NoImage
     mirroring=NoMirror
     resourceName=None
     resourceObject=null
     tiling=NoTile
     imageSize=<64 64>
     margin=<ltrb 0.5 0.5 0.5 0.5>
     outlineSettings {
      bUseBrushTransparency=false
      roundingType=HalfHeightRadius
      width=0
      color {colorUseRule=UseColor_Specified, specifiedColor=<rgba 0 0 0 0>}
      cornerRadii=<0 0 0 0>
     }
     tintColor {colorUseRule=UseColor_Specified, specifiedColor=<rgba 1 1 1 1>}
     uVRegion {bIsValid=false, max=<0 0>, min=<0 0>}
    }
   }
   slot {class=GridSlot, properties {layer=2, padding=<ltrb -6 -6 -6 -6>}}
  }
  Backing {
   widgetClassPath=&Image
   bIsVariable=true
   parent=Root
   widget=Backing
   properties {brush {
     bIsDynamicallyLoaded=false
     drawAs=Box
     imageType=NoImage
     mirroring=NoMirror
     resourceName=None
     resourceObject=null
     tiling=NoTile
     imageSize=<100 100>
     margin=<ltrb 0.5 0.5 0.5 0.5>
     outlineSettings {
      bUseBrushTransparency=false
      roundingType=HalfHeightRadius
      width=0
      color {colorUseRule=UseColor_Specified, specifiedColor=<rgba 0 0 0 0>}
      cornerRadii=<0 0 0 0>
     }
     tintColor {colorUseRule=UseColor_Specified, specifiedColor=<rgba 1 1 1 1>}
     uVRegion {bIsValid=false, max=<0 0>, min=<0 0>}
    }}
   slot {class=GridSlot, properties {layer=3, padding=<ltrb 3 3 3 3>}}
  }
 ]
}
